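# Sound cue for short slide foley on the WittyStork biped cosmetic companion, played as a creature-pet sound.
# A quality-level node feeds the first two levels a random pick from nine slide waves, and the last two a reduced pool of the first four.
# It plays quieter and pitched up, using companion foley attenuation and sharing the biped footstep concurrency set.
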SoundCue /CosmeticCompanions/Assets/Biped/WittyStork/Audio/Foley/SC_MImosa_Biped_WittyStork_Slideshort
root=QualityLevel_0
nodes [
 {id=QualityLevel_0, children [Random_0,Random_0,Random_1,Random_1]}
 {
  id=Random_0
  children [WavePlayer_0,WavePlayer_1,WavePlayer_2,WavePlayer_3,WavePlayer_4,WavePlayer_5,WavePlayer_6,WavePlayer_7,WavePlayer_8]
  properties {weights [1,1,1,1,1,1,1,1,1]}
 }
 {id=Random_1, children [WavePlayer_9, WavePlayer_10, WavePlayer_17, WavePlayer_24], properties {weights [1,1,1,1]}}
 {
  id=WavePlayer_0
  properties {soundWaveAssetPtr=/CosmeticCompanions/Assets/Biped/WittyStork/Audio/Foley/Wavs/Mimosa_Biped_WittyStork_SlideShort_01}
 }
 {
  id=WavePlayer_1
  properties {soundWaveAssetPtr=/CosmeticCompanions/Assets/Biped/WittyStork/Audio/Foley/Wavs/Mimosa_Biped_WittyStork_SlideShort_02}
 }
 {
  id=WavePlayer_10
  properties {soundWaveAssetPtr=&Mimosa_Biped_WittyStork_SlideShort_02}
 }
 {
  id=WavePlayer_17
  properties {soundWaveAssetPtr=/CosmeticCompanions/Assets/Biped/WittyStork/Audio/Foley/Wavs/Mimosa_Biped_WittyStork_SlideShort_03}
 }
 {
  id=WavePlayer_2
  properties {soundWaveAssetPtr=&Mimosa_Biped_WittyStork_SlideShort_03}
 }
 {
  id=WavePlayer_24
  properties {soundWaveAssetPtr=/CosmeticCompanions/Assets/Biped/WittyStork/Audio/Foley/Wavs/Mimosa_Biped_WittyStork_SlideShort_04}
 }
 {
  id=WavePlayer_3
  properties {soundWaveAssetPtr=&Mimosa_Biped_WittyStork_SlideShort_04}
 }
 {
  id=WavePlayer_4
  properties {soundWaveAssetPtr=/CosmeticCompanions/Assets/Biped/WittyStork/Audio/Foley/Wavs/Mimosa_Biped_WittyStork_SlideShort_05}
 }
 {
  id=WavePlayer_5
  properties {soundWaveAssetPtr=/CosmeticCompanions/Assets/Biped/WittyStork/Audio/Foley/Wavs/Mimosa_Biped_WittyStork_SlideShort_06}
 }
 {
  id=WavePlayer_6
  properties {soundWaveAssetPtr=/CosmeticCompanions/Assets/Biped/WittyStork/Audio/Foley/Wavs/Mimosa_Biped_WittyStork_SlideShort_07}
 }
 {
  id=WavePlayer_7
  properties {soundWaveAssetPtr=/CosmeticCompanions/Assets/Biped/WittyStork/Audio/Foley/Wavs/Mimosa_Biped_WittyStork_SlideShort_08}
 }
 {
  id=WavePlayer_8
  properties {soundWaveAssetPtr=/CosmeticCompanions/Assets/Biped/WittyStork/Audio/Foley/Wavs/Mimosa_Biped_WittyStork_SlideShort_09}
 }
 {
  id=WavePlayer_9
  properties {soundWaveAssetPtr=&Mimosa_Biped_WittyStork_SlideShort_01}
 }
]
properties {
 attenuationSettings=/Game/Sounds/Attenuation/AS_Companion_Foley
 maxDistance=1200
 pitchMultiplier=1.3
 priority=0
 soundClassObject=&Fort_Creatures_Pets
 volumeMultiplier=0.35
 concurrencySet [/CosmeticCompanions/Assets/Audio_Shared/Foley/Biped/CONC_Mimosa_Biped_Footstep]
}
